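# Physics asset for the flames attachment mesh (LOD0) of the Beak T1 vehicle body cosmetic.
# It holds a single body made of one large capsule, offset along negative X, raised slightly and tilted.
# It has no constraints, so it defines no joints between bodies.
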PhysicsAsset /VehicleCosmetics/Bodies/Beak_T1/Meshes/ATT_BeakT1_Flames_LOD0_PhysicsAsset
bodies {body [{shapeType=Capsule, length=96.553, radius=78.2913, center=<-58.705 -0.0008 17.9703>, rotation=<rot 0 -89.9994 -60.7512>}]}
constraints []
